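# Material function that animates clay plug chainsaw blades along a looping path through world position offset, toggled by a static switch.
# UV1.x plus time, optionally blended toward a sine back-and-forth, samples a curve atlas row whose RGB drives rotation, length and height.
# Blades rotate about a parameterised axis and pivot and translate along local axes, masked off where UV1.x exceeds 1.
# Rotation-corrected normals are added to incoming material attributes; the pivot debug branch is not wired to the output.
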MaterialFunction /Clyde/Environments/Sets/ClayPlug/TechArt/Chainsaw/MF_ClayPlug_ChainsawBladeWPO
graph {nodes [
  n1 = WorldPosition()
  n2 = Normalize(VectorInput=VectorParameter() {group=Chainsaw Blade WPO, parameterName=ChainsawBlades.Axis.Direction, defaultValue=<rgba 1 0 0 1>}.RGB)
  n3 = Transform(None=@n2) {transformSourceType=TRANSFORMSOURCE_Local}
  n4 = ComponentMask(None=Multiply(
    A=TextureCoordinate() {coordinateIndex=1}
    B=ScalarParameter() {defaultValue=1, group=Chainsaw Blade WPO, parameterName=ChainsawBlades.Spacing}
   )) {r=true}
  n5 = Multiply(A=Time(), B=ScalarParameter() {defaultValue=0.1, group=Chainsaw Blade WPO, parameterName=ChainsawBlades.Speed}) {constB=0.1}
  n6 = Sine(None=Multiply(A=@n5) {constB=10})
  n7 = LinearInterpolate(
   A=@n5
   B=Multiply(
    A=ConstantBiasScale(None=@n6)
    B=ScalarParameter() {defaultValue=0.1, group=Chainsaw Blade WPO, parameterName=ChainsawBlades.BackAndForthDistance}
   )
   Alpha=ScalarParameter() {group=Chainsaw Blade WPO, parameterName=ChainsawBlades.AnimateBackAndForth}
  )
  n8 = CurveAtlasRowParameter(CurveTime=Frac(None=Add(A=@n4, B=@n7))) {
   atlas=/Clyde/Environments/Sets/ClayPlug/TechArt/CA_Clyde
   curve=&/CV_ClayPlug_ChainsawRotationWPO
   desc="R = Rotation, G = Length, B = Height"
   group=Chainsaw Blade WPO
   parameterName=ChainsawBlades.TimingCurve
  }
  n9 = Multiply(A=@n8.G, B=@n3)
  n10 = ScalarParameter() {defaultValue=300, group=Chainsaw Blade WPO, parameterName=ChainsawBlades.Length}
  n11 = Normalize(VectorInput=VectorParameter() {group=Chainsaw Blade WPO, parameterName=ChainsawBlades.Axis.Rotation, defaultValue=<rgba 0 1 0 1>}.RGB)
  n12 = Multiply(
   A=Transform(None=CrossProduct(A=@n2, B=@n11)) {transformSourceType=TRANSFORMSOURCE_Local}
   B=ScalarParameter() {defaultValue=-50, group=Chainsaw Blade WPO, parameterName=ChainsawBlades.Height}
  )
  n13 = Add(
   A=Multiply(A=@n3, B=ScalarParameter() {group=Chainsaw Blade WPO, parameterName=ChainsawBlades.LengthOffset})
   B=Add(A=Multiply(A=@n9, B=@n10), B=Multiply(A=@n8.B, B=@n12))
  )
  Add(A=@n1.XYZ, B=@n13)
  n14 = FunctionInput() {bUsePreviewValueAsDefault=true, inputName=Material, inputType=FunctionInput_MaterialAttributes}
  n15 = GetMaterialAttributes(None=@n14) {attributeGetTypes [#1, #2]}
  n16 = Transform(None=Normalize(VectorInput=VectorParameter() {group=Chainsaw Blade WPO, parameterName=ChainsawBlades.Axis.Rotation, defaultValue=<rgba 0 1 0 1>}.RGB)) {transformSourceType=TRANSFORMSOURCE_Local}
  n17 = VectorParameter() {group=Chainsaw Blade WPO, parameterName=ChainsawBlades.Axis.Pivot, defaultValue=<rgba 0 0 0 1>}
  n18 = TransformPosition(None=Multiply(A=ComponentMask(None=@n17.RGB) {b=true, g=true, r=true}, B=@n17.A)) {transformType=TRANSFORMPOSSOURCE_World}
  n19 = Multiply(
   A=Multiply(A=MaterialFunctionCall() {materialFunction=&ObjectScale}.Scale Z, B=@n8.R)
   B=ScalarParameter() {defaultValue=-1, group=Chainsaw Blade WPO, parameterName=ChainsawBlades.RotationDirection}
  ) {constB=-1}
  n20 = MaterialFunctionCall(Rotation Axis=@n16, Rotation Angle=@n19, Normals Pivot=@n18) {materialFunction=&FixRotateAboutAxisNormals}
  n21 = FunctionInput(Preview=StaticBool()) {
   bUsePreviewValueAsDefault=true
   description=Default False - Use Tangent Space Normals
   inputName=UseWorldSpaceNormals
   inputType=FunctionInput_StaticBool
   sortPriority=1
  }
  n22 = Add(
   A=@n15.Normal
   B=VertexInterpolator(VS=StaticSwitch(True=@n20.New World Space Vertex Normal, False=@n20.New World Space Vertex Normal, Value=@n21)).PS
  )
  n23 = RotateAboutAxis(NormalizedRotationAxis=@n16, RotationAngle=@n19, PivotPoint=@n18, Position=@n1.XYZ)
  n24 = ComponentMask(None=TextureCoordinate() {coordinateIndex=1}) {r=true}
  n25 = Constant()
  n26 = Constant() {r=1}
  n27 = Add(A=@n15.WorldPositionOffset, B=Multiply(A=Add(A=@n23, B=@n13), B=If(A=@n24, "A > B"=@n25, "A == B"=@n26, "A < B"=@n26) {constB=1}))
  n28 = SetMaterialAttributes!(MaterialAttributes=@n15.MaterialAttributes, Normal=@n22, WorldPositionOffset=@n27) {attributeSetTypes [#1, #2]}
  FunctionOutput(None=StaticSwitchParameter(True=@n28, False=@n14) {group=Chainsaw Blade WPO, parameterName=UseChainsawBladeWPO, sortPriority=1})
  n29 = Constant()
  StaticSwitchParameter(
   True=SetMaterialAttributes!(
    MaterialAttributes=@n28
    BaseColor=@n18
    Normal=@n22
    WorldPositionOffset=StaticSwitchParameter(True=@n29, False=@n27) {group=Chainsaw Blade WPO, parameterName=DebugPivotWithoutWPO, sortPriority=50}
   ) {attributeSetTypes [#3, #1, #2]}
   False=@n28
  ) {group=Chainsaw Blade WPO, parameterName=DebugPivot, sortPriority=49}
 ]}
